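# Niagara system for the toilet paper vehicle boost cosmetic, pairing a rotating mesh/sprite "Cone" with a trailing paper ribbon.
# Cone burst-spawns at the boost offset; RibbonPaper streams backward from nearby with quality-scaled spawn rate.
# The ribbon inherits diminishing source movement and gains a growing vertical sine ripple as it ages.
# GameModeID switches offsets and ribbon rate, lifetime and velocity; painted-variant modules recolor both emitters.
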NiagaraSystem /VehicleCosmetics/Boosts/Boost_ToilettePaper/FX/NS_Boost_ToilettePaper
dependencies {
 usedDataInterfaces [&NiagaraDataInterfaceCurve, &NiagaraDataInterfaceSkeletalMesh, &NiagaraDataInterfaceArrayFloat3]
 usedDynamicInputs [&SelectVectorFromArray]
 usedModules [
  &self:SystemSpawnScript.SystemScriptSource.SystemScriptGraph.NiagaraNodeAssignment_0.SetVariables_#1
  &self:SystemSpawnScript.SystemScriptSource.SystemScriptGraph.NiagaraNodeAssignment_1.SetVariables_#2
  &SystemState
  &self:SystemSpawnScript.SystemScriptSource.SystemScriptGraph.NiagaraNodeAssignment_2.SetVariables_#3
 ]
}
emitters {
 Cone {
  emitterSpawnScript [
   /SetVariables(Emitter.No Painted:NiagaraBool=&Boolean_NotOperation(Bool=$User.bIsPainted)) {isSetParameters=true}
   /SetVariables(Emitter.Painted:NiagaraBool=&Boolean_NotOperation(Bool=$StackContext.No Painted)) {isSetParameters=true}
  ]
  emitterUpdateScript [
   &EmitterState(
    Inactive Response=::NewEnumerator1
    Life Cycle Mode=::NewEnumerator1
    Loop Behavior=::NewEnumerator1
    Loop Duration Mode=::NewEnumerator1
    UseLoopDelay=false
   )
   &SpawnBurst_Instantaneous(Use Spawn Probability=false, UseLoopCountLimit=false)
   &SpawnRate(SpawnRate=50, Use Spawn Probability=false) {enabled=false}
  ]
  particleSpawnScript [
   &VC_PaintedVariantsOverride(
    DefaultColor=<rgba 0.4802 0.5 0.3966 1>
    R1 Black:LinearColor=<rgba 0.1921 0.2 0.1586 1>
    R1 BurntSienna:LinearColor=<rgba 0.4802 0.5 0.3966 1>
    R1 Cobalt:LinearColor=<rgba 0.4802 0.5 0.3966 1>
    R1 Crimson:LinearColor=<rgba 0.4802 0.5 0.3966 1>
    R1 DefaultColor=<rgba 0.4802 0.5 0.3966 1>
    R1 ForestGreen:LinearColor=<rgba 0.4802 0.5 0.3966 1>
    R1 Grey:LinearColor=<rgba 0.4802 0.5 0.3966 1>
    R1 Lime:LinearColor=<rgba 0.4802 0.5 0.3966 1>
    R1 Orange:LinearColor=<rgba 0.4802 0.5 0.3966 1>
    R1 Pink:LinearColor=<rgba 0.4802 0.5 0.3966 1>
    R1 Purple:LinearColor=<rgba 0.4802 0.5 0.3966 1>
    R1 Saffron:LinearColor=<rgba 0.4802 0.5 0.3966 1>
    R1 SkyBlue:LinearColor=<rgba 0.4802 0.5 0.3966 1>
    R1 TitaniumWhite:LinearColor=<rgba 0.4802 0.5 0.3966 1>
    R2 Black:LinearColor=<rgba 0.005 0.005 0.005 1>
    R2 BurntSienna:LinearColor=<rgba 0.0729 0.0234 0.0103 1>
    R2 Cobalt:LinearColor=<rgba 0.0119 0.0258 0.2083 1>
    R2 Crimson:LinearColor=<rgba 0.3 0.0094 0.0142 1>
    R2 DefaultColor:LinearColor=<rgba 0.4802 0.5 0.3966 1>
    R2 ForestGreen:LinearColor=<rgba 0.0996 0.3698 0.0162 1>
    R2 Grey:LinearColor=<rgba 0.0677 0.0677 0.0677 1>
    R2 Lime:LinearColor=<rgba 0.3037 0.5312 0 1>
    R2 Orange:LinearColor=<rgba 0.2344 0.0462 0.0085 1>
    R2 Pink:LinearColor=<rgba 0.4 0.04 0.4 1>
    R2 Purple:LinearColor=<rgba 0.136 0.0243 0.3594 1>
    R2 Saffron:LinearColor=<rgba 0.2917 0.2917 0 1>
    R2 SkyBlue:LinearColor=<rgba 0.1916 0.3864 0.776 1>
    Use R1 Painted Overrides=true
    Use R2 Painted Overrides=true
   )
   &InitializeParticle(
    Color=$Particles.Color
    Lifetime=$Emitter.CurrentLoopDuration
    Mesh Scale Mode=::NewEnumerator3
    Mesh Uniform Scale=0.1
    Position Offset=&AddVectorToPosition(
     Position=&MergeHelper_VectorSwitch(0=<-100 0 0>, 1=<-100 0 0>, 2=<-100 0 0>, 3=<-220 0 32>, Index=$User.GameModeID, Num=::NewEnumerator2)
     Vector=&AddVectorToPosition(Position=$User.DTCarOffset, Vector=Linked() {linkedVariable {name=User.SingleBoostPos, type=NiagaraPosition}})
    )
    Position Offset Coordinate Space=::Local
    Sprite Rotation Mode=::NewEnumerator2
    UsePositionOffset=true
    Write Material Random=true
   )
   &AddVelocity(Velocity=<-1500 0 0>) {enabled=false}
  ]
  particleUpdateScript [
   &ParticleState()
   &UpdateMeshOrientation(Rotation Vector=<0 1 0>)
   &SolveForcesAndVelocity(Clamp Velocity=false, Limit Acceleration=false)
  ]
  rendererClasses [&NiagaraSpriteRendererProperties, &NiagaraMeshRendererProperties]
  renderers [
   {rendererClass=&NiagaraSpriteRendererProperties}
   {rendererClass=&NiagaraMeshRendererProperties, rendererIndex=1}
   {rendererClass=&NiagaraMeshRendererProperties, rendererIndex=2}
   {rendererClass=&NiagaraMeshRendererProperties, rendererIndex=3}
  ]
 }
 RibbonPaper {
  emitterUpdateScript [
   &EmitterState()
   &VC_FilteredSocketCount()
   &VC_RateQualityHandler(Default Rate:NiagaraFloat=&MergeHelper_FloatSwitch(0=30, 1=30, 2=30, 3=10, Index=$User.GameModeID, Num=::NewEnumerator2))
   &VC_SpawnRibbon_v1(SpawnRate=$Emitter.Spawn Rate)
  ]
  particleSpawnScript [
   &VC_PaintedVariantsOverride(
    BurntSienna=<rgba 0.3646 0.217 0.1386 1>
    Cobalt=<rgba 0.1248 0.2031 0.5885 1>
    Crimson=<rgba 0.75 0.1211 0.1316 1>
    DefaultColor=<rgba 0.8643 0.9 0.7138 1>
    ForestGreen=<rgba 0.406 0.7969 0.1743 1>
    Grey=<rgba 0.3273 0.3281 0.3135 1>
    Lime=<rgba 0.7118 1 0.1354 1>
    Orange=<rgba 0.6615 0.2944 0.1929 1>
    Pink=<rgba 0.6771 0.1869 0.6771 1>
    Purple=<rgba 0.4095 0.1785 0.6979 1>
    R1 BurntSienna:LinearColor=<rgba 0.0729 0.0288 0.0171 1>
    R1 Cobalt:LinearColor=<rgba 0 0.0185 1 1>
    R1 DefaultColor=<rgba 0.096 0.1 0.0793 1>
    Saffron=<rgba 0.6745 0.6927 0.2714 1>
    SkyBlue=<rgba 0.45 0.6785 0.9 1>
    Use R1 Painted Overrides=true
   )
   &VC_InitializeRibbon_v1(
    Color=$Particles.Color
    Custom Offset=true
    Facing Axis:ENiagaraOrientationAxis=::YAxis
    Lifetime=&MergeHelper_FloatSwitch(0=0.6, 1=0.6, 2=0.6, 3=1, Index=$User.GameModeID, Num=::NewEnumerator2)
    Local Vector Offset=&AddVectorToPosition(
     Position=&MergeHelper_VectorSwitch(0=<-100 3 -28>, 1=<-100 3 -28>, 2=<-100 3 -28>, 3=<-220 3 0>, Index=$User.GameModeID, Num=::NewEnumerator2)
     Vector=&AddVectorToPosition(Position=$User.DTCarOffset, Vector=Linked() {linkedVariable {name=User.SingleBoostPos, type=NiagaraPosition}})
    )
    Offset Amount=1
    Ribbon Width=60
   )
   &AddVelocity(Velocity=&MergeHelper_VectorSwitch(0=<-1500 0 250>, 1=<-1500 0 250>, 2=<-1500 0 250>, 3=<-2400 0 750>, Index=$User.GameModeID, Num=::NewEnumerator2)) {module=AddVelocity001}
   /VehicleCosmetics/FX/VC_FXParents/Modules/DLR_ProductPreviews(GameModeID:NiagaraInt32=$System.GameModeID, bPreviewVehicleSpeed:NiagaraBool=$System.bPreviewVehicleSpeed)
  ]
  particleUpdateScript [
   &ParticleState()
   &InheritSourceMovement(
    Applied Position Scale=&Multiply_VectorByFloat(
     Vector=<1 1 1>
     Float=&FloatFromCurve(FloatCurve=DataInterface() {propertyValues {Curve {keys [{interpMode=RCIM_Cubic, tangentMode=RCTM_Break, time=0.3, value=1}, {interpMode=RCIM_Cubic, tangentMode=RCTM_Break, time=1, value=0}]}}})
    )
    ApplyForce=false
    ApplyPosition=true
    ApplyPositionFalloff=false
   ) {enabled=false, module=InheritSourceMovement001}
   &InheritSourceMovement(
    Applied Position Scale=&VectorFromFloat(Value=&FloatFromCurve(FloatCurve=DataInterface() {propertyValues {Curve {
         defaultValue=1
         keys [
          {arriveTangent=-1.7885, interpMode=RCIM_Cubic, leaveTangent=-1.7885, tangentMode=RCTM_User, time=0.05, value=1}
          {interpMode=RCIM_Cubic, tangentMode=RCTM_User, time=1, value=0}
         ]
        }}}))
    ApplyForce=false
    ApplyPosition=true
    ApplyPositionFalloff=false
   )
   &SolveForcesAndVelocity(Clamp Velocity=false, Limit Acceleration=false)
   /SetVariables(Particles.Position=&AddVectorToPosition(
     Position=$Particles.Position
     Vector=&TransformVector(
      Vector=&MakeVector(
       X=0
       Y=0
       Z=&Sine(
        Normalized Angle=&Add_Float(
         A=&Multiply_Float(A=$Particles.NormalizedAge, B=-1.3)
         B=&Multiply_Float(A=$Emitter.LoopedAge, B=2)
        )
        Period=1
        Scale=&FloatFromCurve(
         FloatCurve=DataInterface() {propertyValues {Curve {keys [{interpMode=RCIM_Cubic, tangentMode=RCTM_Break, time=0, value=0}, {interpMode=RCIM_Cubic, tangentMode=RCTM_Break, time=0.5, value=2}]}}}
         Scale Curve=5
        )
        Bias=0
       )
      )
      Destination Space=::Simulation
      Apply Scale=true
     )
    )) {isSetParameters=true}
   &ScaleColor(
    Scale Alpha=&FloatFromCurve(FloatCurve=DataInterface() {propertyValues {Curve {keys [<at 0.99 1>,<at 1 0>]}}})
    ScaleA=true
    ScaleRGB=true
   )
   &VC_UpdateRibbon_v1()
  ]
  rendererClasses [&NiagaraRibbonRendererProperties]
  renderers [
   {rendererClass=&NiagaraRibbonRendererProperties}
   {rendererClass=&NiagaraRibbonRendererProperties, rendererIndex=1}
  ]
 }
}
summary {
 systemName=NS_Boost_ToilettePaper
 emitters [
  {emitterName=Cone, rendererClasses [&NiagaraSpriteRendererProperties, &NiagaraMeshRendererProperties]}
  {emitterName=RibbonPaper, rendererClasses [&NiagaraRibbonRendererProperties]}
 ]
 userVariables [
  User.bIsFrontendPreview {type=NiagaraBool}
  User.bIsPainted {type=NiagaraBool}
  User.blsIsCoroColor {type=NiagaraBool}
  User.bPreviewVehicleSpeed {type=NiagaraBool}
  User.ColorVariantID {type=NiagaraInt32}
  User.DTCarOffset {type=NiagaraPosition, description="", defaultValue {}}
  User.GameModeID {type=NiagaraInt32, defaultValue=3}
  User.PaintedColor {type=LinearColor, description="", defaultValue {struct=LinearColor, value=<rgba 1 1 1 1>}}
  User.PaintedScalar {type=NiagaraFloat, defaultValue=1, description=""}
  User.PortCount {type=NiagaraInt32, defaultValue=1}
  User.SingleBoostPos {type=NiagaraPosition, description="", defaultValue {}}
 ]
}
